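# Editor utility widget serving as one entry row in a quest editor menu, pairing a button with a wrapped text description.
# InitMenuEntry sets the quest editor button's label and the description text from its inputs.
# Clicking the button broadcasts the OnEditorSelected dispatcher with the entry widget itself, leaving the response to whatever binds to it.
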WidgetBlueprint /Game/NeverCook/EditorUtilities/QuestTools/QuestUtilities/SupportWidgets/QuestEditorMenuEntry
graphs [
 EventGraph {nodes [
   "On Clicked (EditorButton)"!() -> (OutputDelegate, then, widget:User Widget Object Reference, event:Pointer Event Structure) {is=title K2Node_ComponentBoundEvent}
   > Call On Editor Selected!(execute:Exec, self:Quest Editor Menu Entry Object Reference, Widget:User Widget Object Reference=Self-Reference()) -> (then:Exec) {is=title K2Node_CallDelegate}
   Event PreConstruct()
   Event Construct()
   Event Tick()
  ]}
 InitMenuEntry {nodes [
   n1 = InitMenuEntry!() -> (then, ButtonLabel:String, Description:Text) {is=title K2Node_FunctionEntry}
   > setLabel!(
    execute
    self:TA Base Editor Button Widget Object Reference=Get EditorButton() -> (EditorButton:TA Quest Editor Button Object Reference)
    label:String=@n1
   ) -> (then)
   > "SetText (Text)"(self=Get EditorDescription() -> (EditorDescription:Common Text Object Reference), InText=@n1.Description)
  ]}
 OnEditorSelected {nodes [OnEditorSelected!() -> (then, Widget:User Widget Object Reference) {is=title K2Node_FunctionEntry}]}
]
variables [OnEditorSelected {}]
widgets {
 info {inheritedWidgetCount=0, parentClass=&EditorUtilityWidget, rootWidgetClass=&HorizontalBox, widgetCount=4}
 widgets [
  HorizontalBox_0 {widgetClassPath=&HorizontalBox, widget=HorizontalBox_0}
  CanvasPanel_0 {
   widgetClassPath=&CanvasPanel
   parent=HorizontalBox_0
   widget=CanvasPanel_0
   slot {class=HorizontalBoxSlot, properties {horizontalAlignment=HAlign_Left}}
  }
  EditorButton {
   widgetClassPath=/Game/NeverCook/EditorUtilities/QuestTools/SupportWidget/TA_QuestEditor_Button
   bIsVariable=true
   parent=CanvasPanel_0
   widget=EditorButton
   slot {class=CanvasPanelSlot, properties {layoutData {alignment=<0.5 0>, anchors {maximum=<0.5 1>, minimum=<0.5 0>}, offsets=<ltrb 0 0 300 0>}}}
  }
  EditorDescription {
   widgetClassPath=&CommonTextBlock
   bIsVariable=true
   parent=HorizontalBox_0
   widget=EditorDescription
   properties {
    autoWrapText=true
    text=Text
    font {
     bForceMonospaced=false
     bMaterialIsStencil=false
     fontMaterial=null
     fontObject=&BurbankSmall
     letterSpacing=0
     monospacedWidth=1
     size=10
     skewAmount=0
     typefaceFontName=Bold
     outlineSettings {
      bApplyOutlineToDropShadows=false
      bMiteredCorners=false
      bSeparateFillAlpha=false
      outlineMaterial=null
      outlineSize=0
      outlineColor=<rgba 0 0 0 1>
     }
    }
    shadowOffset=<0 0>
   }
   slot {class=HorizontalBoxSlot, properties {verticalAlignment=VAlign_Center, padding=<ltrb 4 3 4 3>, size {sizeRule=Fill, value=1}}}
  }
 ]
}
